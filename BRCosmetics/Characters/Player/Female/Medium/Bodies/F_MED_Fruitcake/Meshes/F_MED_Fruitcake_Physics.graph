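# Physics asset for the F_MED_Fruitcake female player body cosmetic, covering its skirt bones and the lower body around them.
# Six skirt chains (front, side, back; left and right) each link a master bone to three capsule segments with narrow swing and twist limits.
# Pelvis, spine_01 and thighs add body capsules; the lowest front segments use paired capsules for wider coverage.
# Master bones have no constraint to the pelvis, and only the right thigh has one, fully free.
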PhysicsAsset /BRCosmetics/Characters/Player/Female/Medium/Bodies/F_MED_Fruitcake/Meshes/F_MED_Fruitcake_Physics
bodies {
 dyn_skirt_back_01_l [{shapeType=Capsule, length=5.6675, radius=2.1325, shapeName=dyn_L_skirt_D_02_capsule, center=<3.4589 0 0>, rotation=<rot 90 0 0>}]
 dyn_skirt_back_01_r [{shapeType=Capsule, length=5.6675, radius=2.1325, shapeName=dyn_L_skirt_D_02_capsule, center=<3.4589 0 0>, rotation=<rot 90 0 0>}]
 dyn_skirt_back_02_l [{shapeType=Capsule, length=9.0559, radius=2.4441, shapeName=dyn_L_skirt_D_03_capsule, center=<2 0 0>, rotation=<rot 90 0 0>}]
 dyn_skirt_back_02_r [{shapeType=Capsule, length=9.0559, radius=2.4441, shapeName=dyn_L_skirt_D_03_capsule, center=<2 0 0>, rotation=<rot 90 0 0>}]
 dyn_skirt_back_03_l [{shapeType=Capsule, length=9.6411, radius=4.3589, shapeName=dyn_L_skirt_D_03_capsule, center=<4 0 0>, rotation=<rot 90 0 0>}]
 dyn_skirt_back_03_r [{shapeType=Capsule, length=9.6411, radius=4.3589, shapeName=dyn_L_skirt_D_03_capsule, center=<4 0 0>, rotation=<rot 90 0 0>}]
 dyn_skirt_back_mstr_l [{shapeType=Capsule, length=0.3455, radius=1.1982, shapeName=dyn_skirt_frnt_mstr_l_capsule}]
 dyn_skirt_back_mstr_r [{shapeType=Capsule, length=0.3455, radius=1.1982, shapeName=dyn_skirt_frnt_mstr_l_capsule}]
 dyn_skirt_frnt_01_l [{shapeType=Capsule, length=8.7936, radius=2.7064, shapeName=dyn_L_skirt_D_03_capsule, center=<2 0 0>, rotation=<rot 90 0 0>}]
 dyn_skirt_frnt_01_r [{shapeType=Capsule, length=8.7936, radius=2.7064, shapeName=dyn_L_skirt_D_03_capsule, center=<2 0 0>, rotation=<rot 90 0 0>}]
 dyn_skirt_frnt_02_l [{
   shapeType=Capsule
   length=24.0045
   radius=1.7639
   shapeName=dyn_L_skirt_D_03_capsule
   center=<1.5935 0 0>
   rotation=<rot -0.0003 179.9996 -179.9999>
  }]
 dyn_skirt_frnt_02_r [{
   shapeType=Capsule
   length=23.2461
   radius=2.5223
   shapeName=dyn_L_skirt_D_03_capsule
   center=<1.5935 0 0>
   rotation=<rot -0.0003 179.9996 -179.9999>
  }]
 dyn_skirt_frnt_03_l [
  {shapeType=Capsule, length=27.7048, radius=4.0225, shapeName=dyn_skirt_frnt_03_l_capsule, center=<5.213 0 0>}
  {shapeType=Capsule, length=27.7048, radius=3.7253, shapeName=dyn_skirt_frnt_03_l_capsule, center=<-1.4603 0 0>}
 ]
 dyn_skirt_frnt_03_r [
  {shapeType=Capsule, length=27.7048, radius=3.7253, shapeName=dyn_skirt_frnt_03_l_capsule, center=<-1.4603 0 0>}
  {shapeType=Capsule, length=27.7048, radius=3.7253, shapeName=dyn_skirt_frnt_03_l_capsule, center=<5.3121 0 0>}
 ]
 dyn_skirt_frnt_mstr_l [{shapeType=Capsule, length=0.3455, radius=1.1982, shapeName=dyn_skirt_frnt_mstr_l_capsule}]
 dyn_skirt_frnt_mstr_r [{shapeType=Capsule, length=0.3455, radius=1.1982, shapeName=dyn_skirt_frnt_mstr_l_capsule}]
 dyn_skirt_side_01_l [{shapeType=Capsule, length=6.1059, radius=1.694, shapeName=dyn_L_skirt_D_02_capsule, center=<3.4589 0 0>, rotation=<rot 90 0 0>}]
 dyn_skirt_side_01_r [{shapeType=Capsule, length=5.6221, radius=2.1779, shapeName=dyn_L_skirt_D_02_capsule, center=<3.4589 0 0>, rotation=<rot 90 0 0>}]
 dyn_skirt_side_02_l [{
   shapeType=Capsule
   length=8.292
   radius=3.208
   shapeName=dyn_L_skirt_D_03_capsule
   center=<2 0 0>
   rotation=<rot -0.0002 -179.9999 -179.9999>
  }]
 dyn_skirt_side_02_r [{shapeType=Capsule, length=8.292, radius=3.208, shapeName=dyn_L_skirt_D_03_capsule, center=<2 0 0>, rotation=<rot 0.0002 180 179.9999>}]
 dyn_skirt_side_03_l [{shapeType=Capsule, length=18.5349, radius=2.9446, shapeName=dyn_L_skirt_D_03_capsule, center=<2 0 0>, rotation=<rot 0.0005 0 -0.0001>}]
 dyn_skirt_side_03_r [{shapeType=Capsule, length=18.5349, radius=2.9446, shapeName=dyn_L_skirt_D_03_capsule, center=<4 0 0>, rotation=<rot -0.0004 0 0.0001>}]
 dyn_skirt_side_mstr_l [{shapeType=Capsule, length=0.3455, radius=1.1982, shapeName=dyn_skirt_frnt_mstr_l_capsule}]
 dyn_skirt_side_mstr_r [{shapeType=Capsule, length=0.3455, radius=1.1982, shapeName=dyn_skirt_frnt_mstr_l_capsule}]
 pelvis [{shapeType=Capsule, length=5.0953, radius=13.144, shapeName=pelvis_capsule, center=<-4.5385 1.6615 0>}]
 spine_01 [{shapeType=Capsule, length=0.1116, radius=2.4939, shapeName=spine_01_capsule}]
 thigh_l [{
   shapeType=Capsule
   length=20.4096
   radius=9.1009
   shapeName=thigh_l_capsule
   center=<-20.2135 -2.4404 0.6586>
   rotation=<rot 85.4113 179.9988 -0.0005>
  }]
 thigh_r [{
   shapeType=Capsule
   length=20.4096
   radius=9.1009
   shapeName=thigh_l_capsule
   center=<20.2135 -2.4404 0.6586>
   rotation=<rot 85.4113 179.9988 -0.0005>
  }]
}
constraints [
 {bone1Name=dyn_skirt_frnt_02_l, bone2Name=dyn_skirt_frnt_01_l, swing1LimitDegrees=20, swing2LimitDegrees=5, twistLimitDegrees=1}
 {bone1Name=dyn_skirt_frnt_03_l, bone2Name=dyn_skirt_frnt_02_l, swing1LimitDegrees=20, swing2LimitDegrees=5, twistLimitDegrees=1}
 {bone1Name=dyn_skirt_side_03_l, bone2Name=dyn_skirt_side_02_l, swing1LimitDegrees=20, swing2LimitDegrees=20, twistLimitDegrees=1}
 {bone1Name=dyn_skirt_side_02_l, bone2Name=dyn_skirt_side_01_l, swing1LimitDegrees=20, swing2LimitDegrees=20, twistLimitDegrees=1}
 {bone1Name=dyn_skirt_back_02_l, bone2Name=dyn_skirt_back_01_l, swing1LimitDegrees=20, swing2LimitDegrees=5, twistLimitDegrees=1}
 {bone1Name=dyn_skirt_back_03_l, bone2Name=dyn_skirt_back_02_l, swing1LimitDegrees=20, swing2LimitDegrees=5, twistLimitDegrees=1}
 {bone1Name=dyn_skirt_frnt_02_r, bone2Name=dyn_skirt_frnt_01_r, swing1LimitDegrees=20, swing2LimitDegrees=5, twistLimitDegrees=1}
 {bone1Name=dyn_skirt_frnt_03_r, bone2Name=dyn_skirt_frnt_02_r, swing1LimitDegrees=20, swing2LimitDegrees=5, twistLimitDegrees=1}
 {bone1Name=dyn_skirt_side_02_r, bone2Name=dyn_skirt_side_01_r, swing1LimitDegrees=20, swing2LimitDegrees=20, twistLimitDegrees=1}
 {bone1Name=dyn_skirt_side_03_r, bone2Name=dyn_skirt_side_02_r, swing1LimitDegrees=20, swing2LimitDegrees=20, twistLimitDegrees=1}
 {bone1Name=dyn_skirt_back_03_r, bone2Name=dyn_skirt_back_02_r, swing1LimitDegrees=20, swing2LimitDegrees=5, twistLimitDegrees=1}
 {bone1Name=dyn_skirt_back_02_r, bone2Name=dyn_skirt_back_01_r, swing1LimitDegrees=20, swing2LimitDegrees=5, twistLimitDegrees=1}
 {bone1Name=dyn_skirt_frnt_01_l, bone2Name=dyn_skirt_frnt_mstr_l, swing1LimitDegrees=30, swing2LimitDegrees=5, twistLimitDegrees=1}
 {bone1Name=dyn_skirt_side_01_l, bone2Name=dyn_skirt_side_mstr_l, swing1LimitDegrees=20, swing2LimitDegrees=5, twistLimitDegrees=1}
 {bone1Name=dyn_skirt_side_01_r, bone2Name=dyn_skirt_side_mstr_r, swing1LimitDegrees=20, swing2LimitDegrees=5, twistLimitDegrees=1}
 {bone1Name=dyn_skirt_back_01_r, bone2Name=dyn_skirt_back_mstr_r, swing1LimitDegrees=20, swing2LimitDegrees=5, twistLimitDegrees=1}
 {bone1Name=dyn_skirt_back_01_l, bone2Name=dyn_skirt_back_mstr_l, swing1LimitDegrees=20, swing2LimitDegrees=15, twistLimitDegrees=1}
 {bone1Name=dyn_skirt_frnt_01_r, bone2Name=dyn_skirt_frnt_mstr_r, swing1LimitDegrees=30, swing2LimitDegrees=5, twistLimitDegrees=1}
 {
  bone1Name=thigh_r
  bone2Name=pelvis
  swing1LimitDegrees=45
  swing1Motion=Free
  swing2LimitDegrees=45
  swing2Motion=Free
  twistLimitDegrees=45
  twistMotion=Free
 }
]
