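# Post-simulation Control Rig for the Fruitcake female body that corrects skirt and pouch bones.
# Center skirt chains average left and right rotations, the front weighted by SPR_extreme_angle_thigh_l_alpha, the back fully.
# While crouching, a spherical pose reader on thigh_l drives corrective rotations of front skirt bases.
# Pouch bones follow skirt bones through weighted position constraints, get crouch offsets, then aim at dfrm_pouch_aim using an up bone.
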ControlRigBlueprint /BRCosmetics/Characters/Player/Female/Medium/Bodies/F_MED_Fruitcake/Meshes/F_MED_Fruitcake_post_sim_CtrlRig
graphs [
 RigVMFunctionLibrary {links [], nodes []}
 RigVMModel {nodes [
   RigUnit_BeginExecution = &RigUnit_BeginExecution()
   n1 = /Script/ControlRig.RigUnit_SequenceExecution(ExecuteContext=@RigUnit_BeginExecution.ExecutePin) {is=struct RigVMUnitNode}
   n2 = RigVMVariableNode() {variable=SPR_extreme_angle_thigh_l_alpha}
   &RigUnit_RotationConstraintLocalSpaceOffset(
    AdvancedSettings=(RotationOrderForFilter=XZY)
    Child=(Type=Bone,Name="dfrm_skirt_mid_frnt_base")
    Filter=(bX=True,bY=True,bZ=True)
    Parents=((Item=(Type=Bone,Name="dfrm_skirt_frnt_base_l"),Weight=0.5),(Item=(Type=Bone,Name="dfrm_skirt_frnt_base_r"),Weight=0.5))
    ExecutePin=@n1.A
    Weight=@n2.Value
   )
   > &RigUnit_RotationConstraintLocalSpaceOffset(
    AdvancedSettings=(RotationOrderForFilter=XZY)
    Child=(Type=Bone,Name="dyn_skirt_mid_frnt_01")
    Filter=(bX=True,bY=True,bZ=True)
    Parents=((Item=(Type=Bone,Name="dyn_skirt_frnt_01_l"),Weight=0.5),(Item=(Type=Bone,Name="dyn_skirt_frnt_01_r"),Weight=0.5))
    Weight=@n2.Value
   )
   > &RigUnit_RotationConstraintLocalSpaceOffset(
    AdvancedSettings=(RotationOrderForFilter=XZY)
    Child=(Type=Bone,Name="dyn_skirt_mid_frnt_02")
    Filter=(bX=True,bY=True,bZ=True)
    Parents=((Item=(Type=Bone,Name="dyn_skirt_frnt_02_l"),Weight=0.5),(Item=(Type=Bone,Name="dyn_skirt_frnt_02_r"),Weight=0.5))
    Weight=@n2.Value
   )
   > &RigUnit_RotationConstraintLocalSpaceOffset(
    AdvancedSettings=(RotationOrderForFilter=XZY)
    Child=(Type=Bone,Name="dyn_skirt_mid_frnt_03")
    Filter=(bX=True,bY=True,bZ=True)
    Parents=((Item=(Type=Bone,Name="dyn_skirt_frnt_03_l"),Weight=0.5),(Item=(Type=Bone,Name="dyn_skirt_frnt_03_r"),Weight=0.5))
    Weight=@n2.Value
   )
   &RigUnit_RotationConstraintLocalSpaceOffset(
    AdvancedSettings=(RotationOrderForFilter=XZY)
    Child=(Type=Bone,Name="dyn_skirt_mid_back_01")
    Filter=(bX=True,bY=True,bZ=True)
    Parents=((Item=(Type=Bone,Name="dyn_skirt_back_01_l"),Weight=0.5),(Item=(Type=Bone,Name="dyn_skirt_back_01_r"),Weight=0.5))
    Weight=1
    ExecutePin=@n1.B
   )
   > &RigUnit_RotationConstraintLocalSpaceOffset(
    AdvancedSettings=(RotationOrderForFilter=XZY)
    Child=(Type=Bone,Name="dyn_skirt_mid_back_02")
    Filter=(bX=True,bY=True,bZ=True)
    Parents=((Item=(Type=Bone,Name="dyn_skirt_back_02_l"),Weight=0.5),(Item=(Type=Bone,Name="dyn_skirt_back_02_r"),Weight=0.5))
    Weight=1
   )
   > &RigUnit_RotationConstraintLocalSpaceOffset(
    AdvancedSettings=(RotationOrderForFilter=XZY)
    Child=(Type=Bone,Name="dyn_skirt_mid_back_03")
    Filter=(bX=True,bY=True,bZ=True)
    Parents=((Item=(Type=Bone,Name="dyn_skirt_back_03_l"),Weight=0.5),(Item=(Type=Bone,Name="dyn_skirt_back_03_r"),Weight=0.5))
    Weight=1
   )
   Branch_RigVMUnitNode = &RigVMFunction_ControlFlowBranch(ExecuteContext=@n1.C, Condition=RigVMVariableNode() {variable=is_crouching}.Value)
   n3 = &RigUnit_SequenceExecution(ExecuteContext=@Branch_RigVMUnitNode.True)
   n4 = &RigUnit_SphericalPoseReader(
    ActiveRegionScaleFactors=(PositiveWidth=3,NegativeWidth=3,PositiveHeight=1,NegativeHeight=3)
    ActiveRegionSize=0.06
    Debug=(bDrawDebug=false,DebugScale=25,DebugSegments=20,DebugThickness=0.25)
    DriverAxis=(X=-1,Y=0,Z=0)
    DriverItem=(Type=Bone,Name="thigh_l")
    FalloffRegionScaleFactors=(PositiveWidth=1,NegativeWidth=1,PositiveHeight=1,NegativeHeight=1)
    FalloffSize=0.19
    OptionalParentItem=(Type=Bone)
    RotationOffset=(X=80,Y=-10,Z=-15)
    ExecutePin=@n3.A
   )
   > &RigUnit_ModifyTransforms(
    ItemToModify=((Item=(Type=Bone,Name="dfrm_skirt_mid_frnt_base"),Transform=(Rotation=(X=0,Y=0,Z=0,W=1),Translation=(X=0,Y=0,Z=0),Scale3D=(X=1,Y=1,Z=1))))
    ItemToModify.0.Transform.Rotation=&RigVMFunction_MathQuaternionFromEuler(Euler=(X=0,Y=0,Z=-15)).Result
    Weight=@n4.OutputParam
   )
   > &RigUnit_ModifyTransforms(
    ItemToModify=((Item=(Type=Bone,Name="dfrm_skirt_frnt_base_l"),Transform=(Rotation=(X=0,Y=0,Z=0,W=1),Translation=(X=0,Y=0,Z=0),Scale3D=(X=1,Y=1,Z=1))))
    ItemToModify.0.Transform.Rotation=&RigVMFunction_MathQuaternionFromEuler(Euler=(X=0,Y=0,Z=-15)).Result
    Weight=@n4.OutputParam
   )
   > &RigUnit_ModifyTransforms(
    ItemToModify=((Item=(Type=Bone,Name="dfrm_skirt_frnt_base_r"),Transform=(Rotation=(X=0,Y=0,Z=0,W=1),Translation=(X=0,Y=0,Z=0),Scale3D=(X=1,Y=1,Z=1))))
    ItemToModify.0.Transform.Rotation=&RigVMFunction_MathQuaternionFromEuler(Euler=(X=0,Y=0,Z=-10)).Result
    Weight=@n4.OutputParam
   )
   /Script/ControlRig.RigUnit_PositionConstraintLocalSpaceOffset(
    Child=(Type=Bone,Name="dyn_pouch_mstr")
    Filter=(bX=True,bY=True,bZ=True)
    Parents=((Item=(Type=Bone,Name="dfrm_skirt_frnt_base_l"),Weight=0.35),(Item=(Type=Bone,Name="dfrm_skirt_side_base_l"),Weight=0.5),(Item=(Type=Bone,Name="dyn_skirt_frnt_01_l"),Weight=0.017),(Item=(Type=Bone,Name="dyn_skirt_side_01_l"),Weight=0.133))
    Weight=1
    bMaintainOffset=True
    ExecutePin=@n1.D
   ) {is=struct RigVMUnitNode}
   > n5 = &RigUnit_PositionConstraintLocalSpaceOffset(
    Child=(Type=Bone,Name="dfrm_pouch_aim")
    Filter=(bX=True,bY=True,bZ=True)
    Parents="((Item=(Type=Bone,Name=\"dfrm_skirt_frnt_base_l\"),Weight=0.05),(Item=(Type=Bone,Name=\"dyn_skirt_side_02_l\"),Weight=33),(Item=(Type=Bone,Name=\"dyn_skirt_frnt_01_l\"),Weight=0.36),(Item=(Type=Bone,Name=\"dyn_skirt_side_01_l\"),Weight=0.18),(Item=(Type=Bone,Name= ...(truncated)"
    Weight=1
    bMaintainOffset=True
   )
   n6 = &RigUnit_SequenceExecution(ExecuteContext=@n5.ExecutePin)
   Branch_1_RigVMUnitNode = &RigVMFunction_ControlFlowBranch(ExecuteContext=@n6.A, Condition=RigVMVariableNode() {variable=is_crouching}.Value)
   &RigUnit_ModifyTransforms(
    ItemToModify=((Item=(Type=Bone,Name="dyn_pouch_constrained"),Transform=(Rotation=(X=0,Y=0,Z=0,W=1),Translation=(X=-4,Y=1,Z=-5),Scale3D=(X=1,Y=1,Z=1))))
    Weight=1
    ExecutePin=@Branch_1_RigVMUnitNode.True
   )
   > &RigUnit_ModifyTransforms(
    ItemToModify=((Item=(Type=Bone,Name="dfrm_pouch_aim"),Transform=(Rotation=(X=0,Y=0,Z=0,W=1),Translation=(X=-5,Y=-25,Z=-10),Scale3D=(X=1,Y=1,Z=1))))
    Weight=1
   )
   > &RigUnit_ModifyTransforms(
    ItemToModify=((Item=(Type=Bone,Name="dyn_pouch_up"),Transform=(Rotation=(X=0,Y=0,Z=0,W=1),Translation=(X=-6,Y=5,Z=-15),Scale3D=(X=1,Y=1,Z=1))))
    Weight=1
   )
   /Script/ControlRig.RigUnit_AimItem(
    DebugSettings=(Scale=10,WorldOffset=(Rotation=(X=0,Y=0,Z=0,W=1),Translation=(X=0,Y=0,Z=0),Scale3D=(X=1,Y=1,Z=1)))
    Item=(Type=Bone,Name="dyn_pouch_constrained")
    Primary=(Weight=1,Axis=(X=1,Y=0,Z=0),Target=(X=1,Y=0,Z=0),Kind=Location,Space=(Type=Bone,Name="dfrm_pouch_aim"))
    Secondary=(Weight=1,Axis=(X=0,Y=-1,Z=0),Target=(X=1,Y=0,Z=0),Kind=Location,Space=(Type=Bone,Name="dyn_pouch_up"))
    Weight=1
    ExecutePin=@n6.B
   ) {is=struct RigVMUnitNode}
  ]}
]
hierarchy [
 root {loc=<0 0 0>, rot=<0 0 0 1>}
 pelvis {parent=root, loc=<0 -0.5301 98.0571>, rot=<0.0127 0.707 -0.0127 -0.707>}
 spine_01 {parent=pelvis, loc=<6.1208 -0.9328 0>, rot=<0 0 0.034 -0.9994>}
 spine_02 {parent=spine_01, loc=<6.5996 -0.2095 0>, rot=<0 0 0.0238 -0.9997>}
 spine_03 {parent=spine_02, loc=<6.4914 -0.1237 0>, rot=<0 0 0.0149 -0.9999>}
 spine_04 {parent=spine_03, loc=<6.9598 0.0794 0>, rot=<0 0 -0.0132 -0.9999>}
 spine_05 {parent=spine_04, loc=<6.9987 0 0>, rot=<0 0 -0.0015 -1>}
 neck_01 {parent=spine_05, loc=<11.3138 0.1529 0>, rot=<0 0 -0.2003 -0.9797>}
 neck_02 {parent=neck_01, loc=<4.5627 0 0>, rot=<0 0 0.0504 -0.9987>}
 head {parent=neck_02, loc=<5.3914 0 0>, rot=<0 0 0.115 -0.9934>}
 dyn_bangs_strand_a_01 {parent=head, loc=<13.0725 6.1322 -3.3665>, rot=<-0.1891 0.3606 0.8487 0.3374>}
 dyn_bangs_strand_a_02 {parent=dyn_bangs_strand_a_01, loc=<3.4835 0 0>, rot=<0.0276 -0.0018 -0.3001 -0.9535>}
 dyn_bangs_strand_b_01 {parent=head, loc=<13.3345 6.9606 -2.3145>, rot=<-0.0768 0.1595 0.9092 0.377>}
 dyn_bangs_strand_b_02 {parent=dyn_bangs_strand_b_01, loc=<3.3714 0 0>, rot=<-0.0185 -0.0712 -0.251 -0.9652>}
 dyn_bangs_strand_c_01 {parent=head, loc=<13.9358 7.3754 -0.6426>, rot=<0.0127 -0.0881 0.9479 0.3059>}
 dyn_bangs_strand_c_02 {parent=dyn_bangs_strand_c_01, loc=<3.6014 0 0>, rot=<-0.1137 -0.1203 -0.1989 -0.9659>}
 dyn_bangs_strand_d_01 {parent=head, loc=<13.7484 7.1293 0.9737>, rot=<0.1246 -0.2663 0.8774 0.3791>}
 dyn_bangs_strand_d_02 {parent=dyn_bangs_strand_d_01, loc=<3.73 0 0>, rot=<0 0 -0.2863 -0.9581>}
 dyn_bangs_strand_e_01 {parent=head, loc=<13.4421 6.3716 2.3727>, rot=<0.1897 -0.415 0.8034 0.3825>}
 dyn_bangs_strand_e_02 {parent=dyn_bangs_strand_e_01, loc=<3.6466 0 0>, rot=<-0.678 -0.2007 -0.2007 -0.678>}
 dyn_hair_strand_a_01_r {parent=head, loc=<9.6701 6.6446 6.7874>, rot=<0.0967 -0.5148 0.8517 0.0154>}
 dyn_hair_strand_a_02_r {parent=dyn_hair_strand_a_01_r, loc=<8.9293 0 0>, rot=<0.0021 -0.0558 -0.0684 -0.9961>}
 dyn_hair_strand_a_03_r {parent=dyn_hair_strand_a_02_r, loc=<8.9293 0 0>, rot=<0.0025 -0.023 -0.0066 -0.9997>}
 dyn_hair_strand_b_01_r {parent=head, loc=<9.9001 4.7157 6.9619>, rot=<0.1451 -0.766 0.6262 -0.0091>}
 dyn_hair_strand_b_02_r {parent=dyn_hair_strand_b_01_r, loc=<8.144 0 0>, rot=<-0.0276 0.1321 -0.0543 -0.9894>}
 dyn_hair_strand_b_03_r {parent=dyn_hair_strand_b_02_r, loc=<8.144 0 0>, rot=<-0.0012 0.1216 0.0644 -0.9905>}
 dyn_hair_strand_01_l {parent=head, loc=<3.231 -2.6969 -7.3357>, rot=<0.0837 0.3013 0.9493 -0.0325>}
 dyn_hair_strand_02_l {parent=dyn_hair_strand_01_l, loc=<5.3089 0 0>, rot=<0.0017 0.0071 0.14 -0.9901>}
 dyn_hair_strand_03_l {parent=dyn_hair_strand_02_l, loc=<5.3089 0 0>, rot=<0.1364 0.0356 0.1398 -0.9801>}
 dyn_ponytail_base {parent=head, loc=<14.6837 -4.7671 -0.0232>, rot=<0.0041 0 0 -1>}
 dyn_ponytail_cd_01 {parent=dyn_ponytail_base, loc=<-0.1853 -7.5678 -0.4721>, rot=<0.0562 -0.9981 0.0027 0.0237>}
 dyn_ponytail_c_01 {parent=dyn_ponytail_cd_01, loc=<9.2023 0.2007 -0.9082>, rot=<0.0563 -0.0303 -0.1046 -0.9925>}
 dyn_ponytail_c_02 {parent=dyn_ponytail_c_01, loc=<8.0332 0 0>, rot=<-0.0494 0.0347 0.0871 -0.9944>}
 dyn_ponytail_d_01 {parent=dyn_ponytail_cd_01, loc=<9.2454 -0.2007 0.9082>, rot=<0.2171 -0.0062 -0.0534 -0.9747>}
 dyn_ponytail_d_02 {parent=dyn_ponytail_d_01, loc=<9.2558 0 0>, rot=<-0.0002 0.0468 -0.0032 -0.9989>}
 dyn_ponytail_a_01 {parent=dyn_ponytail_base, loc=<-11.4068 -3.9296 4.7847>, rot=<-0.0028 -0.9847 0.1591 0.0707>}
 dyn_ponytail_a_02 {parent=dyn_ponytail_a_01, loc=<7.5018 0 0>, rot=<0 0 0 1>}
 dyn_ponytail_e_01 {parent=dyn_ponytail_base, loc=<-11.1801 -3.5684 0.9643>, rot=<-0.0096 -0.9995 -0.0036 0.0312>}
 dyn_ponytail_e_02 {parent=dyn_ponytail_e_01, loc=<6.7158 0 0>, rot=<0 0 0 1>}
 dyn_ponytail_f_01 {parent=dyn_ponytail_base, loc=<-17.38 -6.329 -4.4697>, rot=<-0.0302 -0.9994 -0.0047 -0.0153>}
 dyn_ponytail_g_01 {parent=dyn_ponytail_base, loc=<-17.1408 -5.9079 -2.4591>, rot=<0.0342 -0.9991 -0.0048 0.0239>}
 dyn_ponytail_hi_01 {parent=dyn_ponytail_base, loc=<-15.5197 -3.2907 -5.6266>, rot=<0.0285 -0.9559 -0.2821 -0.0763>}
 dyn_ponytail_h_01 {parent=dyn_ponytail_hi_01, loc=<6.3863 -0.2464 1.3019>, rot=<-0.0332 -0.002 0.0704 -0.997>}
 dyn_ponytail_i_01 {parent=dyn_ponytail_hi_01, loc=<4.9921 0.1916 -1.474>, rot=<-0.2778 -0.0773 -0.0249 -0.9572>}
 dyn_earing_r {parent=head, loc=<0.186 -0.3913 6.8273>, rot=<0.5021 -0.4979 0.5021 0.4979>}
 dyn_earing_l {parent=head, loc=<0.2087 -0.0575 -6.8006>, rot=<0.6256 -0.6205 -0.3357 -0.333>}
 clavicle_l {parent=spine_05, loc=<5.4831 -0.1053 -3.2946>, rot=<0.0465 -0.7289 -0.009 -0.6829>}
 upperarm_l {parent=clavicle_l, loc=<9.8658 -1.2136 -0.4105>, rot=<0.0164 -0.3245 -0.0269 -0.9454>}
 lowerarm_l {parent=upperarm_l, loc=<26.465 -0.8345 0.259>, rot=<0.0036 0.0161 -0.2298 -0.9731>}
 hand_l {parent=lowerarm_l, loc=<22.7355 -0.3203 -0.6105>, rot=<-0.6784 -0.0223 0.0735 -0.7307>}
 thumb_01_l {parent=hand_l, loc=<1.3118 -0.3191 -2.459>, rot=<0.6546 -0.4042 0.0664 -0.6354>}
 thumb_02_l {parent=thumb_01_l, loc=<5.9336 -0.0448 0.0401>, rot=<-0.0037 0.0213 0.298 -0.9543>}
 thumb_03_l {parent=thumb_02_l, loc=<2.2223 0.3682 -0.0907>, rot=<0.0118 -0.013 0.0066 -0.9998>}
 index_metacarpal_l {parent=hand_l, loc=<1.6535 0.0183 -2.96>, rot=<0.0158 -0.0124 -0.0173 -0.9996>}
 index_01_l {parent=index_metacarpal_l, loc=<8.3008 0.0357 -0.0524>, rot=<-0.0044 0.0485 0.0924 -0.9945>}
 index_02_l {parent=index_01_l, loc=<3.3587 -0.3004 0.0452>, rot=<-0.002 0.0027 0.3225 -0.9466>}
 index_03_l {parent=index_02_l, loc=<2.5578 -0.123 -0.0252>, rot=<-0.002 0.0067 0.2558 -0.9667>}
 middle_metacarpal_l {parent=hand_l, loc=<2.1856 -0.1725 -0.5984>, rot=<0.006 0.0168 -0.0329 -0.9993>}
 middle_01_l {parent=middle_metacarpal_l, loc=<7.3275 -0.2447 0.0129>, rot=<-0.0208 0.0042 0.0858 -0.9961>}
 middle_02_l {parent=middle_01_l, loc=<4.0123 0.1401 0.1175>, rot=<-0.0218 0.0258 0.3512 -0.9357>}
 middle_03_l {parent=middle_02_l, loc=<2.6077 0.0534 0.1373>, rot=<0.0082 0.0188 0.1708 -0.9851>}
 ring_metacarpal_l {parent=hand_l, loc=<2.2074 -0.3625 1.3638>, rot=<-0.0284 0.0722 -0.0458 -0.9959>}
 ring_01_l {parent=ring_metacarpal_l, loc=<7.0295 -0.6197 -0.1446>, rot=<-0.0489 -0.0252 0.1892 -0.9804>}
 ring_02_l {parent=ring_01_l, loc=<3.6174 0.4647 -0.1027>, rot=<0 0 0.2806 -0.9598>}
 ring_03_l {parent=ring_02_l, loc=<2.4152 0.114 0.0199>, rot=<0 0 0.1189 -0.9929>}
 pinky_metacarpal_l {parent=hand_l, loc=<1.7616 -0.9753 3.3814>, rot=<-0.0397 0.0814 -0.069 -0.9935>}
 pinky_01_l {parent=pinky_metacarpal_l, loc=<7.1709 -0.4831 -0.1094>, rot=<-0.0654 -0.0035 0.2242 -0.9723>}
 pinky_02_l {parent=pinky_01_l, loc=<2.8274 0.0893 -0.1043>, rot=<0.0037 0.0011 0.2854 -0.9584>}
 pinky_03_l {parent=pinky_02_l, loc=<2.2046 0.0307 -0.0078>, rot=<-0.0002 0.0098 0.1445 -0.9895>}
 deform_wrist_dn_base_l {parent=hand_l, loc=<0 -1.5326 0>, rot=<0 0 0 -1>}
 deform_wrist_dn_l {parent=deform_wrist_dn_base_l, loc=<0 0 0>, rot=<0 0 0 1>}
 deform_wrist_up_base_l {parent=hand_l, loc=<0 2.3009 0>, rot=<0 0 0 -1>}
 deform_wrist_up_l {parent=deform_wrist_up_base_l, loc=<0 0 0>, rot=<0 0 0 1>}
 weapon_l {parent=hand_l, loc=<-0.1974 -3.4111 0.3811>, rot=<0 0 0 1>}
 hand_attach_l {parent=hand_l, loc=<8.5 -3 0>, rot=<0.7071 0.7071 0 0>}
 dyn_handPetal_01_l {parent=hand_l, loc=<1.9461 3.5581 -0.5884>, rot=<0.6739 0.1692 -0.2605 -0.6704>}
 dyn_handPetal_02_l {parent=hand_l, loc=<1.8496 1.2664 3.668>, rot=<0.2326 0.3103 0.0127 -0.9216>}
 dyn_handPetal_03_l {parent=hand_l, loc=<1.0111 -2.9303 2.6559>, rot=<-0.3217 0.3887 0.3241 -0.8003>}
 dyn_handPetal_04_l {parent=hand_l, loc=<0.2061 -3.4964 -2.2213>, rot=<0.7341 -0.2574 -0.4424 0.4463>}
 dyn_handPetal_05_l {parent=hand_l, loc=<0.7523 0.5516 -4.289>, rot=<0.8856 -0.0614 -0.4155 -0.1984>}
 lowerarm_twist_02_l {parent=lowerarm_l, loc=<10.6707 0 0>, rot=<0 0 0 1>}
 lowerarm_twist_01_l {parent=lowerarm_l, loc=<18.316 0 0>, rot=<0 0 0 1>}
 deform_elbow_in_base_l {parent=lowerarm_l, loc=<0.1834 3.6062 -0.2572>, rot=<0 0 0 1>}
 deform_elbow_in_l {parent=deform_elbow_in_base_l, loc=<0 0 0>, rot=<0 0 0 1>}
 deform_elbow_out_base_l {parent=lowerarm_l, loc=<0 -4 0>, rot=<0 0 0 1>}
 deform_elbow_out_l {parent=deform_elbow_out_base_l, loc=<0 0 0>, rot=<0 0 0 1>}
 upperarm_twist_01_l {parent=upperarm_l, loc=<7.2135 -0.1134 0>, rot=<0 0 0 1>}
 deform_upperarm_bk_base_l {parent=upperarm_twist_01_l, loc=<-6.4053 -4.3746 0.314>, rot=<0 0 0 1>}
 deform_upperarm_bk_l {parent=deform_upperarm_bk_base_l, loc=<0 0 0>, rot=<0 0 0 1>}
 deform_upperarm_fr_base_l {parent=upperarm_twist_01_l, loc=<-5.6312 5.092 0.3354>, rot=<0 0 0 1>}
 deform_upperarm_fr_l {parent=deform_upperarm_fr_base_l, loc=<0 0 0>, rot=<0 0 0 1>}
 deform_upperarm_up_base_l {parent=upperarm_twist_01_l, loc=<-8.891 0.4825 3.9517>, rot=<0 0 0 1>}
 deform_upperarm_up_l {parent=deform_upperarm_up_base_l, loc=<0 0 0>, rot=<0 0 0 1>}
 upperarm_twist_02_l {parent=upperarm_l, loc=<19.7998 -0.1134 0>, rot=<0 0 0 1>}
 clavicle_r {parent=spine_05, loc=<5.4831 -0.1053 3.2946>, rot=<0.7289 0.0465 -0.6829 0.009>}
 upperarm_r {parent=clavicle_r, loc=<-9.8658 1.2136 0.4105>, rot=<0.0164 -0.3245 -0.0269 -0.9454>}
 lowerarm_r {parent=upperarm_r, loc=<-26.465 0.8345 -0.259>, rot=<0.0036 0.0161 -0.2298 -0.9731>}
 hand_r {parent=lowerarm_r, loc=<-22.7355 0.3203 0.6106>, rot=<-0.6784 -0.0223 0.0735 -0.7307>}
 thumb_01_r {parent=hand_r, loc=<-1.3118 0.3191 2.459>, rot=<0.6546 -0.4042 0.0664 -0.6354>}
 thumb_02_r {parent=thumb_01_r, loc=<-5.9336 0.0448 -0.0401>, rot=<-0.0037 0.0213 0.298 -0.9543>}
 thumb_03_r {parent=thumb_02_r, loc=<-2.2223 -0.3682 0.0907>, rot=<0.0118 -0.013 0.0066 -0.9998>}
 index_metacarpal_r {parent=hand_r, loc=<-1.6535 -0.0183 2.96>, rot=<0.0158 -0.0124 -0.0173 -0.9996>}
 index_01_r {parent=index_metacarpal_r, loc=<-8.3008 -0.0357 0.0524>, rot=<-0.0044 0.0485 0.0924 -0.9945>}
 index_02_r {parent=index_01_r, loc=<-3.3587 0.3004 -0.0452>, rot=<-0.002 0.0027 0.3225 -0.9466>}
 index_03_r {parent=index_02_r, loc=<-2.5578 0.123 0.0252>, rot=<-0.002 0.0067 0.2558 -0.9667>}
 middle_metacarpal_r {parent=hand_r, loc=<-2.1856 0.1725 0.5984>, rot=<0.006 0.0168 -0.0329 -0.9993>}
 middle_01_r {parent=middle_metacarpal_r, loc=<-7.3275 0.2447 -0.0129>, rot=<-0.0208 0.0042 0.0858 -0.9961>}
 middle_02_r {parent=middle_01_r, loc=<-4.0123 -0.1401 -0.1175>, rot=<-0.0218 0.0258 0.3512 -0.9357>}
 middle_03_r {parent=middle_02_r, loc=<-2.6077 -0.0534 -0.1373>, rot=<0.0082 0.0188 0.1708 -0.9851>}
 ring_metacarpal_r {parent=hand_r, loc=<-2.2074 0.3625 -1.3638>, rot=<-0.0284 0.0722 -0.0458 -0.9959>}
 ring_01_r {parent=ring_metacarpal_r, loc=<-7.0295 0.6197 0.1446>, rot=<-0.0489 -0.0252 0.1892 -0.9804>}
 ring_02_r {parent=ring_01_r, loc=<-3.6174 -0.4647 0.1027>, rot=<0 0 0.2806 -0.9598>}
 ring_03_r {parent=ring_02_r, loc=<-2.4152 -0.114 -0.0199>, rot=<0 0 0.1189 -0.9929>}
 pinky_metacarpal_r {parent=hand_r, loc=<-1.7616 0.9753 -3.3814>, rot=<-0.0397 0.0814 -0.069 -0.9935>}
 pinky_01_r {parent=pinky_metacarpal_r, loc=<-7.1709 0.4831 0.1094>, rot=<-0.0654 -0.0035 0.2242 -0.9723>}
 pinky_02_r {parent=pinky_01_r, loc=<-2.8274 -0.0893 0.1043>, rot=<0.0037 0.0011 0.2854 -0.9584>}
 pinky_03_r {parent=pinky_02_r, loc=<-2.2046 -0.0307 0.0078>, rot=<-0.0002 0.0098 0.1445 -0.9895>}
 deform_wrist_dn_base_r {parent=hand_r, loc=<-0.0001 1.5328 0>, rot=<0 0 0 -1>}
 deform_wrist_dn_r {parent=deform_wrist_dn_base_r, loc=<0 0 0>, rot=<0 0 0 1>}
 deform_wrist_up_base_r {parent=hand_r, loc=<-0.0002 -2.3007 -0.0001>, rot=<0 0 0 -1>}
 deform_wrist_up_r {parent=deform_wrist_up_base_r, loc=<0 0 0>, rot=<0 0 0 1>}
 weapon_r {parent=hand_r, loc=<0.1974 3.4111 -0.3811>, rot=<0 0 0 1>}
 hand_attach_r {parent=hand_r, loc=<-8.5001 2.9999 0>, rot=<0 0 -0.7071 -0.7071>}
 dyn_handPetal_05_r {parent=hand_r, loc=<-0.7521 -0.5517 4.2891>, rot=<0.8856 -0.0614 -0.4155 -0.1984>}
 dyn_handPetal_04_r {parent=hand_r, loc=<-0.2059 3.4962 2.2213>, rot=<0.7341 -0.2574 -0.4424 0.4463>}
 dyn_handPetal_03_r {parent=hand_r, loc=<-1.0108 2.93 -2.6558>, rot=<-0.3217 0.3887 0.3241 -0.8003>}
 dyn_handPetal_02_r {parent=hand_r, loc=<-1.8493 -1.2667 -3.668>, rot=<0.2326 0.3103 0.0127 -0.9216>}
 dyn_handPetal_01_r {parent=hand_r, loc=<-1.9461 -3.5582 0.5884>, rot=<0.6739 0.1692 -0.2605 -0.6704>}
 lowerarm_twist_02_r {parent=lowerarm_r, loc=<-10.6707 0 0>, rot=<0 0 0 -1>}
 lowerarm_twist_01_r {parent=lowerarm_r, loc=<-18.316 0 0>, rot=<0 0 0 -1>}
 deform_elbow_in_base_r {parent=lowerarm_r, loc=<-0.1834 -3.6062 0.2572>, rot=<0 0 0 -1>}
 deform_elbow_in_r {parent=deform_elbow_in_base_r, loc=<0 0 0>, rot=<0 0 0 1>}
 deform_elbow_out_base_r {parent=lowerarm_r, loc=<0 4 -0.0001>, rot=<0 0 0 -1>}
 deform_elbow_out_r {parent=deform_elbow_out_base_r, loc=<0 0 0>, rot=<0 0 0 1>}
 upperarm_twist_01_r {parent=upperarm_r, loc=<-7.2135 0.1134 0>, rot=<0 0 0 -1>}
 deform_upperarm_bk_base_r {parent=upperarm_twist_01_r, loc=<6.4053 4.3746 -0.314>, rot=<0 0 0 1>}
 deform_upperarm_bk_r {parent=deform_upperarm_bk_base_r, loc=<0 0 0>, rot=<0 0 0 1>}
 deform_upperarm_fr_base_r {parent=upperarm_twist_01_r, loc=<5.6312 -5.092 -0.3354>, rot=<0 0 0 1>}
 deform_upperarm_fr_r {parent=deform_upperarm_fr_base_r, loc=<0 0 0>, rot=<0 0 0 1>}
 deform_upperarm_up_base_r {parent=upperarm_twist_01_r, loc=<8.891 -0.4825 -3.9517>, rot=<0 0 0 1>}
 deform_upperarm_up_r {parent=deform_upperarm_up_base_r, loc=<0 0 0>, rot=<0 0 0 1>}
 upperarm_twist_02_r {parent=upperarm_r, loc=<-19.7998 0.1134 0>, rot=<0 0 0 -1>}
 dyn_necklace_mstr {parent=spine_05, loc=<7.1199 4.9352 0>, rot=<0 0 0.9225 0.386>}
 dyn_necklace_01 {parent=dyn_necklace_mstr, loc=<0 0 0>, rot=<0 0 0 1>}
 dyn_necklace_02 {parent=dyn_necklace_01, loc=<2.3892 0 0>, rot=<-0.7065 -0.0291 -0.0291 -0.7065>}
 attach_backpack {parent=spine_05, loc=<0 -8.35 0>, rot=<-0.0093 -0.707 0.0093 -0.707>}
 attach_cape {parent=spine_05, loc=<9.3216 -8.4501 0>, rot=<0.0616 -0.7044 -0.0616 -0.7044>}
 deform_pec_aim_base_l {parent=spine_04, loc=<3.2456 9.3309 -1.2237>, rot=<0.0603 -0.4033 0.3318 -0.8507>}
 deform_pec_aim_l {parent=deform_pec_aim_base_l, loc=<0 0 0>, rot=<0 0 0 1>}
 deform_pec_base_l {parent=deform_pec_aim_l, loc=<9.3665 0 0>, rot=<0 0 0 1>}
 deform_pec_l {parent=deform_pec_base_l, loc=<0 0 0>, rot=<0 0 0 1>}
 deform_pec_aim_base_r {parent=spine_04, loc=<3.2452 9.3309 1.2237>, rot=<-0.4033 -0.0603 0.8507 0.3318>}
 deform_pec_aim_r {parent=deform_pec_aim_base_r, loc=<0 0 0>, rot=<0 0 0 1>}
 deform_pec_base_r {parent=deform_pec_aim_r, loc=<-9.3667 -0.0003 -0.0004>, rot=<0 0 0 1>}
 deform_pec_r {parent=deform_pec_base_r, loc=<0 0 0>, rot=<0 0 0 1>}
 dyn_pouch_mstr {parent=spine_01, loc=<3.2377 5.5761 -12.7806>, rot=<-0.1397 0.5403 0.794 0.2411>}
 dyn_pouch_constrained {parent=dyn_pouch_mstr, loc=<0 0 0>, rot=<0 0 0 1>}
 dyn_pouch {parent=dyn_pouch_constrained, loc=<0 0 0>, rot=<0 0 0 1>}
 dyn_pouch_up {parent=dyn_pouch_mstr, loc=<0 -7.76 0>, rot=<0 0 0 1>}
 thigh_l {parent=pelvis, loc=<-1.0855 -0.2109 -8.7053>, rot=<0.0009 0.0355 0.0155 -0.9993>}
 calf_l {parent=thigh_l, loc=<-42.5284 -0.0019 -0.7297>, rot=<0.0006 -0.0161 -0.0391 -0.9991>}
 foot_l {parent=calf_l, loc=<-46.7401 -2.5014 -0.592>, rot=<0.0427 -0.0238 0.0353 -0.9982>}
 ball_l {parent=foot_l, loc=<-3.5658 10.8066 -0.1544>, rot=<0.0001 -0.0001 -0.7063 -0.7079>}
 calf_twist_02_l {parent=calf_l, loc=<-19.1546 -1.7943 -0.2638>, rot=<0.0237 0.0095 -0.0035 -0.9997>}
 calf_twist_01_l {parent=calf_l, loc=<-35.8483 -1.9197 -0.5794>, rot=<0.0237 0.0095 -0.0035 -0.9997>}
 deform_knee_in_base_l {parent=calf_l, loc=<0.8031 -5.0231 0.2456>, rot=<0 0 0 -1>}
 deform_knee_in_l {parent=deform_knee_in_base_l, loc=<0 0 0>, rot=<0 0 0 1>}
 deform_knee_out_base_l {parent=calf_l, loc=<0 4 0>, rot=<0 0 0 -1>}
 deform_knee_out_l {parent=deform_knee_out_base_l, loc=<0 0 0>, rot=<0 0 0 1>}
 thigh_twist_01_l {parent=thigh_l, loc=<-27.952 0 0>, rot=<0.0013 0 0 -1>}
 deform_groin_base_l {parent=thigh_l, loc=<0.4801 5.6776 -0.2486>, rot=<0 0 0 -1>}
 deform_groin_l {parent=deform_groin_base_l, loc=<0 0 0>, rot=<0 0 0 1>}
 thigh_r {parent=pelvis, loc=<-1.0855 -0.2109 8.7053>, rot=<0.0355 -0.0009 0.9993 0.0155>}
 calf_r {parent=thigh_r, loc=<42.5284 0.0019 0.7297>, rot=<0.0006 -0.0161 -0.0391 -0.9991>}
 foot_r {parent=calf_r, loc=<46.7401 2.5014 0.592>, rot=<0.0427 -0.0238 0.0353 -0.9982>}
 ball_r {parent=foot_r, loc=<3.5658 -10.8066 0.1544>, rot=<0.0001 -0.0001 -0.7063 -0.7079>}
 calf_twist_02_r {parent=calf_r, loc=<19.1546 1.7943 0.2638>, rot=<0.0237 0.0095 -0.0035 -0.9997>}
 calf_twist_01_r {parent=calf_r, loc=<35.8483 1.9197 0.5794>, rot=<0.0237 0.0095 -0.0035 -0.9997>}
 deform_knee_in_base_r {parent=calf_r, loc=<-0.8031 5.0231 -0.2457>, rot=<0 0 0 -1>}
 deform_knee_in_r {parent=deform_knee_in_base_r, loc=<0 0 0>, rot=<0 0 0 1>}
 deform_knee_out_base_r {parent=calf_r, loc=<0 -4 0>, rot=<0 0 0 -1>}
 deform_knee_out_r {parent=deform_knee_out_base_r, loc=<0 0 0>, rot=<0 0 0 1>}
 thigh_twist_01_r {parent=thigh_r, loc=<27.952 0 0>, rot=<0.0013 0 0 -1>}
 deform_groin_base_r {parent=thigh_r, loc=<-0.4801 -5.6776 0.2486>, rot=<0 0 0 -1>}
 deform_groin_r {parent=deform_groin_base_r, loc=<0 0 0>, rot=<0 0 0 1>}
 deform_glute_base_l {parent=pelvis, loc=<-7.3466 -9.0213 -9.1245>, rot=<0.0172 0.031 -0.4477 -0.8935>}
 deform_glute_l {parent=deform_glute_base_l, loc=<0 0 0>, rot=<0 0 0 1>}
 deform_glute_base_r {parent=pelvis, loc=<-7.3464 -9.0213 9.1245>, rot=<0.031 -0.0172 0.8935 -0.4477>}
 deform_glute_r {parent=deform_glute_base_r, loc=<0 0 0>, rot=<0 0 0 1>}
 dfrm_pouch_aim {parent=pelvis, loc=<-1.7625 8.4246 -19.4854>, rot=<-0.0127 -0.707 0.0127 -0.707>}
 dfrm_skirt_mstr {parent=pelvis, loc=<12.6773 -1.7077 0>, rot=<0 0 0.0598 -0.9982>}
 dfrm_skirt_frnt_base_l {parent=dfrm_skirt_mstr, loc=<0.5126 7.982 -6.548>, rot=<-0.1169 0.3676 0.9002 0.2023>}
 dyn_skirt_frnt_mstr_l {parent=dfrm_skirt_frnt_base_l, loc=<10 0 0>, rot=<0 0 0 1>}
 dyn_skirt_frnt_01_l {parent=dyn_skirt_frnt_mstr_l, loc=<0 0 0>, rot=<0 0 0 1>}
 dyn_skirt_frnt_02_l {parent=dyn_skirt_frnt_01_l, loc=<10 0 0>, rot=<0 0 -0.1117 -0.9937>}
 dyn_skirt_frnt_03_l {parent=dyn_skirt_frnt_02_l, loc=<10 0 0>, rot=<0 0 -0.0846 -0.9964>}
 dfrm_skirt_frnt_base_r {parent=dfrm_skirt_mstr, loc=<0.5126 7.982 6.548>, rot=<0.1169 -0.3676 0.9002 0.2023>}
 dyn_skirt_frnt_mstr_r {parent=dfrm_skirt_frnt_base_r, loc=<10 0 0>, rot=<0 0 0 1>}
 dyn_skirt_frnt_01_r {parent=dyn_skirt_frnt_mstr_r, loc=<0 0 0>, rot=<0 0 0 1>}
 dyn_skirt_frnt_02_r {parent=dyn_skirt_frnt_01_r, loc=<9.9999 0 0>, rot=<0 0 -0.1117 -0.9937>}
 dyn_skirt_frnt_03_r {parent=dyn_skirt_frnt_02_r, loc=<10 0 0>, rot=<0 0 -0.0846 -0.9964>}
 dfrm_skirt_side_base_l {parent=dfrm_skirt_mstr, loc=<1.064 1.3932 -9.3427>, rot=<-0.213 0.6766 0.6876 0.1548>}
 dyn_skirt_side_mstr_l {parent=dfrm_skirt_side_base_l, loc=<10 0 0>, rot=<0 0 0 1>}
 dyn_skirt_side_01_l {parent=dyn_skirt_side_mstr_l, loc=<0 0 0>, rot=<0 0 0 1>}
 dyn_skirt_side_02_l {parent=dyn_skirt_side_01_l, loc=<10 0 0>, rot=<0 0 -0.1117 -0.9937>}
 dyn_skirt_side_03_l {parent=dyn_skirt_side_02_l, loc=<10 0 0>, rot=<0 0 -0.0846 -0.9964>}
 dfrm_skirt_back_base_l {parent=dfrm_skirt_mstr, loc=<1.686 -5.3717 -6.348>, rot=<0.2762 -0.8829 -0.3677 -0.0951>}
 dyn_skirt_back_mstr_l {parent=dfrm_skirt_back_base_l, loc=<10 0 0>, rot=<0 0 0 1>}
 dyn_skirt_back_01_l {parent=dyn_skirt_back_mstr_l, loc=<0 0 0>, rot=<0 0 0 1>}
 dyn_skirt_back_02_l {parent=dyn_skirt_back_01_l, loc=<10 0 0>, rot=<0 0 -0.1117 -0.9937>}
 dyn_skirt_back_03_l {parent=dyn_skirt_back_02_l, loc=<10 0 0>, rot=<0 0 -0.0846 -0.9964>}
 dfrm_skirt_mid_back_base {parent=dfrm_skirt_mstr, loc=<1.8405 -7.8576 0>, rot=<0.3004 -0.9538 0.0035 -0.0011>}
 dyn_skirt_mid_back_mstr {parent=dfrm_skirt_mid_back_base, loc=<10 0 0>, rot=<0 0 0 1>}
 dyn_skirt_mid_back_01 {parent=dyn_skirt_mid_back_mstr, loc=<0 0 0>, rot=<0 0 0 1>}
 dyn_skirt_mid_back_02 {parent=dyn_skirt_mid_back_01, loc=<10 0 0>, rot=<0 0 -0.1117 -0.9937>}
 dyn_skirt_mid_back_03 {parent=dyn_skirt_mid_back_02, loc=<10 0 0>, rot=<0 0 -0.0846 -0.9964>}
 dfrm_skirt_side_base_r {parent=dfrm_skirt_mstr, loc=<1.0643 1.3933 9.3428>, rot=<0.213 -0.6766 0.6876 0.1548>}
 dyn_skirt_side_mstr_r {parent=dfrm_skirt_side_base_r, loc=<10 0 0>, rot=<0 0 0 -1>}
 dyn_skirt_side_01_r {parent=dyn_skirt_side_mstr_r, loc=<0 0 0>, rot=<0 0 0 -1>}
 dyn_skirt_side_02_r {parent=dyn_skirt_side_01_r, loc=<10.0003 0 0>, rot=<0 0 -0.1117 -0.9937>}
 dyn_skirt_side_03_r {parent=dyn_skirt_side_02_r, loc=<10 0 0>, rot=<0 0 -0.0846 -0.9964>}
 dfrm_skirt_back_base_r {parent=dfrm_skirt_mstr, loc=<1.6865 -5.3718 6.3481>, rot=<0.2762 -0.8829 0.3677 0.0951>}
 dyn_skirt_back_mstr_r {parent=dfrm_skirt_back_base_r, loc=<10 0 0>, rot=<0 0 0 -1>}
 dyn_skirt_back_01_r {parent=dyn_skirt_back_mstr_r, loc=<0 0 0>, rot=<0 0 0 -1>}
 dyn_skirt_back_02_r {parent=dyn_skirt_back_01_r, loc=<10.0004 0 0>, rot=<0 0 -0.1117 -0.9937>}
 dyn_skirt_back_03_r {parent=dyn_skirt_back_02_r, loc=<9.9999 0 0>, rot=<0 0 -0.0846 -0.9964>}
 dfrm_skirt_mid_frnt_mstr {parent=dfrm_skirt_mstr, loc=<0.2834 10.6586 0>, rot=<0.0008 0.0036 0.9756 0.2194>}
 dfrm_skirt_mid_frnt_base {parent=dfrm_skirt_mid_frnt_mstr, loc=<0 0 0>, rot=<0 0 0 1>}
 dyn_skirt_mid_frnt_mstr {parent=dfrm_skirt_mid_frnt_base, loc=<10 0 0>, rot=<0 0 0 1>}
 dyn_skirt_mid_frnt_01 {parent=dyn_skirt_mid_frnt_mstr, loc=<0 0 0>, rot=<0 0 0 1>}
 dyn_skirt_mid_frnt_02 {parent=dyn_skirt_mid_frnt_01, loc=<10 0 0>, rot=<0 0 -0.1117 -0.9937>}
 dyn_skirt_mid_frnt_03 {parent=dyn_skirt_mid_frnt_02, loc=<10 0 0>, rot=<0 0 -0.0846 -0.9964>}
 attach {parent=root, loc=<0 0 0>, rot=<0 0 0 1>}
 ik_foot_root {parent=root, loc=<0 0 0>, rot=<0 0 0 1>}
 ik_foot_l {parent=ik_foot_root, loc=<14.8262 -7.3545 8.296>, rot=<0.0357 0.7045 0.0284 -0.7082>}
 ik_foot_r {parent=ik_foot_root, loc=<-14.8262 -7.3545 8.296>, rot=<0.7082 0.0284 -0.7045 0.0357>}
 ik_hand_root {parent=root, loc=<0 0 0>, rot=<0 0 0 1>}
 ik_hand_gun {parent=ik_hand_root, loc=<-48.1743 3.3545 104.4381>, rot=<0.6276 0.1309 0.4183 -0.6434>}
 ik_hand_r {parent=ik_hand_gun, loc=<0 0 0>, rot=<0 0 0 1>}
 ik_hand_l {parent=ik_hand_gun, loc=<59.3271 67.6997 34.3524>, rot=<-0.6158 -0.7027 -0.3565 0>}
 SPR_extreme_angle_thigh_l_alpha {type=Curve, loc=<0 0 0>, rot=<0 0 0 1>}
]
variables [
 SPR_extreme_angle_thigh_l_alpha {type=double, default=0}
 is_crouching {type=bool, default=False}
]
