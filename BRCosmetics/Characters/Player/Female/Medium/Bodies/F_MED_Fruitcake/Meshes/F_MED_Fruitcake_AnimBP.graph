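# Post-process style AnimBP for the F_MED_Fruitcake female body, adding secondary motion to a pose copied from the attached parent mesh.
# Deformation and movement-driven Control Rigs feed two spring AnimDynamics necklace bones, then a cached pose.
# A left-thigh extreme-angle curve switches between standard and crouch-asset RigidBody sims, followed by a post-sim Control Rig.
# The event graph updates movement alphas, that switch and gravity override, using a deprecated helper.
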AnimBlueprint /BRCosmetics/Characters/Player/Female/Medium/Bodies/F_MED_Fruitcake/Meshes/F_MED_Fruitcake_AnimBP
target_skeleton=/Game/Characters/Player/Female/Medium/Bodies/F_MED_Fruitcake/Meshes/F_MED_Fruitcake_Skeleton
graphs [
 AnimGraph {nodes [
   n1 = RigidBody!(
    ComponentPose=Local To Component(LocalPose=Use cached pose 'pre_sim_cache'())
    Alpha=Get SPR_extreme_angle_thigh_l_rev_alpha() -> (SPR_extreme_angle_thigh_l_rev_alpha:double)
    OverrideWorldGravity:Vector=Get gravity_override() -> (gravity_override:Vector)
    ExternalForce
   ) -> (Pose) {
    bClampLinearTranslationLimitToRefPose=true
    bOverrideWorldGravity=true
    overridePhysicsAsset=&/F_MED_Fruitcake_Physics
    simulationSpace=BaseBoneSpace
    alphaBoolBlend {blendOption=Linear}
    baseBoneRef {boneName=pelvis}
    componentAppliedLinearAccClamp=<3000 3000 3000>
    componentLinearAccScale=<0.3 0.3 0.3>
    componentLinearVelScale=<0.3 0.3 0.3>
    overrideWorldGravity {z=-980}
    simSpaceSettings {maxAngularAcceleration=500, maxAngularVelocity=3, maxLinearAcceleration=250, maxLinearVelocity=500, velocityScaleZ=0, worldAlpha=1}
   }
   n2 = RigidBody(
    ComponentPose=Local To Component(LocalPose=Use cached pose 'pre_sim_cache'())
    Alpha=Get SPR_extreme_angle_thigh_l_alpha() -> (SPR_extreme_angle_thigh_l_alpha:double)
    OverrideWorldGravity=Get gravity_override()
   ) {
    bClampLinearTranslationLimitToRefPose=true
    bOverrideWorldGravity=true
    overridePhysicsAsset=&/F_MED_Fruitcake_crouch_Physics
    simulationSpace=BaseBoneSpace
    alphaBoolBlend {blendOption=Linear}
    baseBoneRef {boneName=pelvis}
    componentAppliedLinearAccClamp=<3000 3000 3000>
    componentLinearAccScale=<0.3 0.3 0.3>
    componentLinearVelScale=<0.3 0.3 0.3>
    overrideWorldGravity {z=-980}
    simSpaceSettings {maxAngularAcceleration=500, maxAngularVelocity=3, maxLinearAcceleration=250, maxLinearVelocity=500, velocityScaleZ=0, worldAlpha=1}
   }
   n3 = Control Rig!(
    Source=Blend Poses by bool(
     BlendPose_0=Component To Local(ComponentPose=@n2)
     BlendPose_1=Component To Local(ComponentPose=@n1)
     bActiveValue=Get SPR_extreme_angle_thigh_l_bool() -> (SPR_extreme_angle_thigh_l_bool:bool)
    ) {blendType=Linear}
    Alpha
    is_crouching:bool=Get bIsCrouching()
    SPR_extreme_angle_thigh_l_alpha:double=Get SPR_extreme_angle_thigh_l_rev_alpha()
   ) -> (Pose) {
    alphaBoolBlend {blendOption=Linear}
    controlRigAssetReference {blueprintRigClass=&/F_MED_Fruitcake_post_sim_CtrlRig.F_MED_Fruitcake_post_sim_CtrlRig_C}
   }
   Output Pose(Result=@n3) {layerGroup=None}
   n4 = Control Rig!(Source=Copy Pose From Mesh() {bUseAttachedParent=true}, Alpha) -> (Pose) {
    alphaBoolBlend {blendOption=Linear}
    controlRigAssetReference {blueprintRigClass=&/F_MED_Fruitcake_Deformation_ControlRig.F_MED_Fruitcake_Deformation_ControlRig_C}
   }
   n5 = Control Rig!(
    Source=@n4
    Alpha
    forwards_backwards_alpha:double=Get forwards_backwards_alpha() -> (forwards_backwards_alpha:double)
    is_crouching:bool=Get bIsCrouching()
    is_crouching_alpha:double=Get is_crouching_alpha() -> (is_crouching_alpha:double)
    is_crouch_sprinting:bool=Get bIsCrouchSprinting()
    is_sprinting:double=Get is_sprinting_alpha() -> (is_sprinting_alpha:double)
    is_jumping:double=Get is_jumping_alpha() -> (is_jumping_alpha:double)
    left_right_alpha:double=Get left_right_alpha() -> (left_right_alpha:double)
    is_DBNO:bool=Get bIsDBNO()
    is_sliding_alpha:double=Get is_sliding_alpha() -> (is_sliding_alpha:double)
   ) -> (Pose) {
    alphaBoolBlend {blendOption=Linear}
    controlRigAssetReference {blueprintRigClass=&/F_MED_Fruitcake_CtrlRig.F_MED_Fruitcake_CtrlRig_C}
   }
   n6 = Anim Dynamics - Bone: dyn_necklace_01(ComponentPose=Local To Component(LocalPose=@n5), Alpha=1, ExternalForce="0,0,0") {
    angularDampingOverride=0.9999
    angularSpringConstant=500
    bAngularSpring=true
    bOverrideAngularDamping=true
    bUseGravityOverride=true
    simulationSpace=RootRelative
    alphaBoolBlend {blendOption=Linear}
    boundBone {boneName=dyn_necklace_01}
    componentAppliedLinearAccClamp=<3000 3000 3000>
    componentLinearAccScale=<0.1 0.1 0.1>
    componentLinearVelScale=<0.9 0.9 0.9>
    gravityOverride {z=-980}
    physicsBodyDefinitions [{
      boundBone {boneName=dyn_necklace_01}
      boxExtents=<10 10 10>
      constraintSetup {angularLimitsMax=<0 5 0>, angularLimitsMin=<0 -5 -20>, angularTarget=<0 0 0>}
      localJointOffset=<1 0 0>
     }]
   }
   n7 = Anim Dynamics - Bone: dyn_necklace_02(ComponentPose=@n6, Alpha=1, ExternalForce="0,0,0") {
    angularDampingOverride=0.99
    angularSpringConstant=1500
    bAngularSpring=true
    bOverrideAngularDamping=true
    bUseGravityOverride=true
    simulationSpace=RootRelative
    alphaBoolBlend {blendOption=Linear}
    boundBone {boneName=dyn_necklace_02}
    componentAppliedLinearAccClamp=<3000 3000 3000>
    componentLinearAccScale=<0.1 0.1 0.1>
    componentLinearVelScale=<0.9 0.9 0.9>
    gravityOverride {z=-980}
    physicsBodyDefinitions [{
      boundBone {boneName=dyn_necklace_02}
      boxExtents=<5 5 5>
      constraintSetup {angularLimitsMax=<0 5 0>, angularLimitsMin=<0 -5 -10>, angularTarget=<0 0 0>}
      localJointOffset=<1 0 0>
     }]
   }
   Save cached pose 'pre_sim_cache'(Pose=Component To Local(ComponentPose=@n7)) {cacheName=pre_sim_cache}
  ]}
 EventGraph {nodes [
   Event BlueprintUpdateAnimation()
   > n1 = Sequence!(execute) -> (then_0, then_1, then_2:Exec, then_3:Exec)
   Set is_crouching_alpha(execute=@n1.then_2, is_crouching_alpha:double="To Float (Boolean)"(InBool=Get bIsCrouching())) -> (Output_Get:double)
   > Set is_sprinting_alpha(is_sprinting_alpha:double="To Float (Boolean)"(InBool=Get bIsSprinting())) -> (Output_Get:double)
   > Set is_jumping_alpha(is_jumping_alpha:double="To Float (Boolean)"(InBool=Get bIsJumping())) -> (Output_Get:double)
   > Set is_sliding_alpha(is_sliding_alpha:double="To Float (Boolean)"(InBool=Get bIsSliding())) -> (Output_Get:double)
   n2 = Gravity Override State Function DEPRECATED(execute=@n1.then_0, CharacterPartAnimInstance=Self-Reference(), GameTimeInSeconds=GetGameTimeInSeconds())
   > Set gravity_override(gravity_override=@n2)
   n3 = PlayerMovement(execute=@n1.then_1, AnimInstance=Self-Reference(), XDotProductCoefficient=1, YDotProductCoefficient=1, ZDotProductCoefficient=1)
   > Set left_right_alpha(left_right_alpha:double=@n3) -> (Output_Get:double)
   > Set forwards_backwards_alpha(forwards_backwards_alpha:double=@n3.YMovement) -> (Output_Get:double)
   n4 = Branch(execute=@n1.then_3, Condition="float > float"(A=GetCurveValue(CurveName=SPR_extreme_angle_thigh_l_alpha)))
   > Set SPR_extreme_angle_thigh_l_bool(SPR_extreme_angle_thigh_l_bool:bool=true) -> (Output_Get:bool)
   > Set SPR_extreme_angle_thigh_l_alpha(SPR_extreme_angle_thigh_l_alpha:double=1) -> (Output_Get:double)
   > Set SPR_extreme_angle_thigh_l_rev_alpha(SPR_extreme_angle_thigh_l_rev_alpha:double=0) -> (Output_Get:double)
   Set SPR_extreme_angle_thigh_l_bool(execute=@n4.else, SPR_extreme_angle_thigh_l_bool=false)
   > Set SPR_extreme_angle_thigh_l_alpha(SPR_extreme_angle_thigh_l_alpha=0)
   > Set SPR_extreme_angle_thigh_l_rev_alpha(SPR_extreme_angle_thigh_l_rev_alpha=1)
  ]}
]
variables [
 forwards_backwards_alpha {type=number}
 gravity_override {type=Vector}
 is_crouching_alpha {type=number}
 is_jumping_alpha {type=number}
 is_sliding_alpha {type=number}
 is_sprinting_alpha {type=number}
 left_right_alpha {type=number}
 SPR_extreme_angle_thigh_l_alpha {type=number}
 SPR_extreme_angle_thigh_l_bool {type=boolean}
 SPR_extreme_angle_thigh_l_rev_alpha {type=number}
]
