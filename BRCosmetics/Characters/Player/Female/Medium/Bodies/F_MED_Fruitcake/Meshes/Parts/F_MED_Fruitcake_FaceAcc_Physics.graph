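# Physics asset for the face accessory part of the female medium Fruitcake character, simulating dynamic hair strands and a multi-branch ponytail.
# Capsule bodies cover the hair bones plus head, neck, upper spine, clavicles and upper arms.
# Limited swing and tight twist constraints chain three side strands to the head and ponytail branches to a base body, some via shared intermediate bodies.
# Only the right arm chain has constraints among torso and arms, all free-moving.
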PhysicsAsset /BRCosmetics/Characters/Player/Female/Medium/Bodies/F_MED_Fruitcake/Meshes/Parts/F_MED_Fruitcake_FaceAcc_Physics
bodies {
 clavicle_l [{
   shapeType=Capsule
   length=10.1018
   radius=4.2502
   shapeName=clavicle_l_capsule
   center=<3.2789 0 1.3035>
   rotation=<rot -73.4188 179.9991 -179.9993>
  }]
 clavicle_r [{
   shapeType=Capsule
   length=10.1018
   radius=4.2502
   shapeName=clavicle_l_capsule
   center=<-3.2789 0 -1.3035>
   rotation=<rot -73.4188 179.9991 0.0007>
  }]
 dyn_hair_strand_01_l [{
   shapeType=Capsule
   length=5.6151
   radius=0.9562
   shapeName=dyn_hair_strand_01_l_capsule
   center=<2.473 0 0>
   rotation=<rot 90 160.5289 160.5291>
  }]
 dyn_hair_strand_02_l [{
   shapeType=Capsule
   length=5.6151
   radius=0.9562
   shapeName=dyn_hair_strand_02_l_capsule
   center=<2.473 0 0>
   rotation=<rot 90 178.3124 178.3126>
  }]
 dyn_hair_strand_03_l [{
   shapeType=Capsule
   length=5.6151
   radius=0.9562
   shapeName=dyn_hair_strand_03_l_capsule
   center=<2.473 0 0>
   rotation=<rot 90 -173.2786 -173.2784>
  }]
 dyn_hair_strand_a_01_r [{
   shapeType=Capsule
   length=8.8884
   radius=1.0978
   shapeName=dyn_hair_strand_a_01_r_capsule
   center=<3.5736 0 0>
   rotation=<rot 90 144.7357 144.7359>
  }]
 dyn_hair_strand_a_02_r [{
   shapeType=Capsule
   length=8.8884
   radius=1.0978
   shapeName=dyn_hair_strand_a_02_r_capsule
   center=<3.5736 0 0>
   rotation=<rot 90 179.9472 179.9474>
  }]
 dyn_hair_strand_a_03_r [{
   shapeType=Capsule
   length=8.8884
   radius=1.0978
   shapeName=dyn_hair_strand_a_03_r_capsule
   center=<3.5736 0 0>
   rotation=<rot 90 -178.7341 -178.734>
  }]
 dyn_hair_strand_b_01_r [{shapeType=Capsule, length=7.9535, radius=1.193, shapeName=dyn_hair_strand_b_01_r_capsule, center=<3.8896 0 0>, rotation=<rot 90 90 90>}]
 dyn_hair_strand_b_02_r [{shapeType=Capsule, length=7.9535, radius=1.193, shapeName=dyn_hair_strand_b_02_r_capsule, center=<3.8896 0 0>, rotation=<rot 90 0.0001 0>}]
 dyn_hair_strand_b_03_r [{
   shapeType=Capsule
   length=7.9535
   radius=1.193
   shapeName=dyn_hair_strand_b_03_r_capsule
   center=<3.8896 0 0>
   rotation=<rot 90 179.9802 179.9802>
  }]
 dyn_ponytail_a_01 [{shapeType=Capsule, length=6.52, radius=1.3022, shapeName=dyn_ponytail_a_01_capsule, center=<3.3223 0 0>, rotation=<rot 90 23.2459 23.2454>}]
 dyn_ponytail_a_02 [{shapeType=Capsule, length=6.52, radius=1.3022, shapeName=dyn_ponytail_a_02_capsule, center=<3.3223 0 0>, rotation=<rot 90 0 -0.0004>}]
 dyn_ponytail_base [{shapeType=Capsule, length=2.3481, radius=2.4088, shapeName=dyn_ponytail_base_capsule}]
 dyn_ponytail_c_01 [{shapeType=Capsule, length=6.0674, radius=1.3968, shapeName=dyn_ponytail_c_01_capsule, center=<3.3143 0 0>, rotation=<rot -90 -6.7214 6.7217>}]
 dyn_ponytail_c_02 [{
   shapeType=Capsule
   length=6.0674
   radius=1.3968
   shapeName=dyn_ponytail_c_02_capsule
   center=<3.3143 0 0>
   rotation=<rot -90 -179.9999 -179.9998>
  }]
 dyn_ponytail_cd_01 [{
   shapeType=Capsule
   length=5.716
   radius=2.1568
   shapeName=dyn_ponytail_cd_01_capsule
   center=<4.3519 0 0>
   rotation=<rot 90 178.7341 178.7336>
  }]
 dyn_ponytail_d_01 [{
   shapeType=Capsule
   length=9.4731
   radius=1.278
   shapeName=dyn_ponytail_d_01_capsule
   center=<3.6396 0 0>
   rotation=<rot 90 179.0504 179.0497>
  }]
 dyn_ponytail_d_02 [{
   shapeType=Capsule
   length=9.4731
   radius=1.278
   shapeName=dyn_ponytail_d_02_capsule
   center=<3.6396 0 0>
   rotation=<rot 90 -179.6836 -179.6843>
  }]
 dyn_ponytail_e_01 [{shapeType=Capsule, length=6.1443, radius=3.3715, shapeName=dyn_ponytail_e_01_capsule, center=<3.985 0 0>}]
 dyn_ponytail_e_02 [{shapeType=Capsule, length=8.32, radius=2.2707, shapeName=dyn_ponytail_e_02_capsule, center=<9.3904 -2.7243 0>}]
 dyn_ponytail_f_01 [{
   shapeType=Capsule
   length=10.1649
   radius=1.3422
   shapeName=dyn_ponytail_f_01_capsule
   center=<5.8132 0 0>
   rotation=<rot 90 -166.7374 -166.7378>
  }]
 dyn_ponytail_g_01 [{shapeType=Capsule, length=12.0496, radius=1.3422, shapeName=dyn_ponytail_g_01_capsule, center=<6.0056 0 0>, rotation=<rot 90 90 89.9996>}]
 dyn_ponytail_h_01 [{
   shapeType=Capsule
   length=6.644
   radius=0.9502
   shapeName=dyn_ponytail_h_01_capsule
   center=<2.8135 0 0>
   rotation=<rot 90 -179.9602 -179.9611>
  }]
 dyn_ponytail_hi_01 [{shapeType=Capsule, length=2.8066, radius=2.709, shapeName=dyn_ponytail_hi_01_capsule}]
 dyn_ponytail_i_01 [{
   shapeType=Capsule
   length=5.7876
   radius=1.8067
   shapeName=dyn_ponytail_i_01_capsule
   center=<1.6008 0 0>
   rotation=<rot 90 -179.9998 179.9994>
  }]
 head [
  {shapeType=Capsule, length=2.4851, radius=1.5569, shapeName=head_capsule}
  {shapeType=Capsule, length=6.9303, radius=6.9206, shapeName=head_capsule, center=<2.6858 3.8381 0>, rotation=<rot 90 179.9993 179.9996>}
  {shapeType=Capsule, length=20.1962, radius=1.8321, shapeName=head_capsule, center=<-1.9946 -9.2986 0>}
 ]
 neck_02 [{shapeType=Capsule, length=5.7713, radius=4.453, shapeName=neck_02_capsule}]
 spine_05 [
  {shapeType=Capsule, length=16.0746, radius=9.9079, shapeName=spine_05_capsule, center=<-0.5819 0 0>}
  {shapeType=Capsule, length=13.7574, radius=7.6468, shapeName=spine_05_capsule, center=<7.6827 2.7549 0>}
 ]
 upperarm_l [{
   shapeType=Capsule
   length=24.9968
   radius=5.0032
   shapeName=upperarm_l_capsule
   center=<13.1558 -0.2873 -0.7381>
   rotation=<rot -90 179.9736 -179.9737>
  }]
 upperarm_r [{
   shapeType=Capsule
   length=24.9968
   radius=5.0032
   shapeName=upperarm_l_capsule
   center=<-13.1558 0.2873 0.7381>
   rotation=<rot -90 0.0198 179.9801>
  }]
}
constraints [
 {bone1Name=dyn_hair_strand_02_l, bone2Name=dyn_hair_strand_01_l, swing1LimitDegrees=20, swing2LimitDegrees=20, twistLimitDegrees=2}
 {bone1Name=dyn_hair_strand_01_l, bone2Name=head, swing1LimitDegrees=5, swing2LimitDegrees=5, twistLimitDegrees=2}
 {bone1Name=dyn_hair_strand_a_03_r, bone2Name=dyn_hair_strand_a_02_r, swing1LimitDegrees=20, swing2LimitDegrees=20, twistLimitDegrees=2}
 {bone1Name=dyn_hair_strand_a_02_r, bone2Name=dyn_hair_strand_a_01_r, swing1LimitDegrees=20, swing2LimitDegrees=20, twistLimitDegrees=2}
 {bone1Name=dyn_hair_strand_a_01_r, bone2Name=head, swing1LimitDegrees=20, swing2LimitDegrees=20, twistLimitDegrees=2}
 {bone1Name=dyn_hair_strand_b_03_r, bone2Name=dyn_hair_strand_b_02_r, swing1LimitDegrees=20, swing2LimitDegrees=20, twistLimitDegrees=2}
 {bone1Name=dyn_hair_strand_b_02_r, bone2Name=dyn_hair_strand_b_01_r, swing1LimitDegrees=20, swing2LimitDegrees=20, twistLimitDegrees=2}
 {bone1Name=dyn_hair_strand_b_01_r, bone2Name=head, swing1LimitDegrees=20, swing2LimitDegrees=20, twistLimitDegrees=2}
 {bone1Name=dyn_hair_strand_03_l, bone2Name=dyn_hair_strand_02_l, swing1LimitDegrees=20, swing2LimitDegrees=20, twistLimitDegrees=2}
 {
  bone1Name=upperarm_r
  bone2Name=clavicle_r
  swing1LimitDegrees=45
  swing1Motion=Free
  swing2LimitDegrees=45
  swing2Motion=Free
  twistLimitDegrees=45
  twistMotion=Free
 }
 {bone1Name=dyn_ponytail_c_02, bone2Name=dyn_ponytail_c_01, swing1LimitDegrees=20, swing2LimitDegrees=20, twistLimitDegrees=2}
 {bone1Name=dyn_ponytail_c_01, bone2Name=dyn_ponytail_cd_01, swing1LimitDegrees=20, swing2LimitDegrees=20, twistLimitDegrees=5}
 {bone1Name=dyn_ponytail_d_02, bone2Name=dyn_ponytail_d_01, swing1LimitDegrees=20, swing2LimitDegrees=20, twistLimitDegrees=5}
 {bone1Name=dyn_ponytail_d_01, bone2Name=dyn_ponytail_cd_01, swing1LimitDegrees=20, swing2LimitDegrees=20, twistLimitDegrees=5}
 {bone1Name=dyn_ponytail_a_02, bone2Name=dyn_ponytail_a_01, swing1LimitDegrees=20, swing2LimitDegrees=20, twistLimitDegrees=5}
 {bone1Name=dyn_ponytail_e_02, bone2Name=dyn_ponytail_e_01, swing1LimitDegrees=20, swing2LimitDegrees=2, twistLimitDegrees=2}
 {bone1Name=dyn_ponytail_i_01, bone2Name=dyn_ponytail_hi_01, swing1LimitDegrees=40, swing2LimitDegrees=20, twistLimitDegrees=5}
 {bone1Name=dyn_ponytail_h_01, bone2Name=dyn_ponytail_hi_01, swing1LimitDegrees=40, swing2LimitDegrees=20, twistLimitDegrees=5}
 {
  bone1Name=clavicle_r
  bone2Name=spine_05
  swing1LimitDegrees=45
  swing1Motion=Free
  swing2LimitDegrees=45
  swing2Motion=Free
  twistLimitDegrees=45
  twistMotion=Free
 }
 {bone1Name=dyn_ponytail_cd_01, bone2Name=dyn_ponytail_base, swing1LimitDegrees=30, swing2LimitDegrees=20, twistLimitDegrees=5}
 {bone1Name=dyn_ponytail_hi_01, bone2Name=dyn_ponytail_base, swing1LimitDegrees=40, swing2LimitDegrees=10, twistLimitDegrees=2}
 {bone1Name=dyn_ponytail_a_01, bone2Name=dyn_ponytail_base, swing1LimitDegrees=20, swing2LimitDegrees=20, twistLimitDegrees=2}
 {bone1Name=dyn_ponytail_e_01, bone2Name=dyn_ponytail_base, swing1LimitDegrees=40, swing2LimitDegrees=10, twistLimitDegrees=2}
 {bone1Name=dyn_ponytail_g_01, bone2Name=dyn_ponytail_base, swing1LimitDegrees=40, swing2LimitDegrees=20, twistLimitDegrees=2}
 {bone1Name=dyn_ponytail_f_01, bone2Name=dyn_ponytail_base, swing1LimitDegrees=40, swing2LimitDegrees=20, twistLimitDegrees=2}
]
